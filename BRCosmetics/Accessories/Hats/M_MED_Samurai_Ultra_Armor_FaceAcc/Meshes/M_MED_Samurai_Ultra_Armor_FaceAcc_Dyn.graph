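# Physics asset for the Samurai Ultra Armor face accessory cosmetic, giving dynamic motion to its chin guard, cheek flaps, tassels and hair bones.
# All bodies are capsules, with head and spine_05 as body volumes and four capsules shaping the chin guard.
# Chin guard (a one-axis hinge), flaps and individual hair bones constrain to the head, which is angularly locked to spine_05.
# Two-link tassel chains hang from the chin guard and head with limited swing and twist.
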PhysicsAsset /BRCosmetics/Accessories/Hats/M_MED_Samurai_Ultra_Armor_FaceAcc/Meshes/M_MED_Samurai_Ultra_Armor_FaceAcc_Dyn
bodies {
 dyn_cheek_flap_l [{shapeType=Capsule, length=6.25, radius=2, center=<0.5 0 0>, rotation=<rot -90 0 0>}]
 dyn_cheek_flap_r [{shapeType=Capsule, length=6.25, radius=2, center=<-0.5 0 0>, rotation=<rot -90 0 0>}]
 dyn_cheek_tassel_01_l [{shapeType=Capsule, length=1.75, radius=1.25, center=<1.4754 0 0>, rotation=<rot -90 0.0099 -0.0099>}]
 dyn_cheek_tassel_01_r [{shapeType=Capsule, length=1.75, radius=1.25, center=<-1.4754 0 0>, rotation=<rot -90 0.0099 -0.0099>}]
 dyn_cheek_tassel_02_l [{shapeType=Capsule, length=2.0181, radius=1.7319, center=<4.7638 0.0001 0>, rotation=<rot -90 0 0>}]
 dyn_cheek_tassel_02_r [{shapeType=Capsule, length=2.0181, radius=1.7319, center=<-4.7638 0 0>, rotation=<rot -90 0 0>}]
 dyn_chin_guard [
  {shapeType=Capsule, length=18.75, radius=1, shapeName=dyn_chin_guard_capsule, center=<0.001 -2.111 0>}
  {shapeType=Capsule, length=20, radius=1, shapeName=dyn_chin_guard_capsule, center=<1.9869 -1.2515 0>}
  {shapeType=Capsule, length=20, radius=1, shapeName=dyn_chin_guard_capsule, center=<7.5892 -0.4021 0>}
  {shapeType=Capsule, length=21.25, radius=1, shapeName=dyn_chin_guard_capsule, center=<4.0411 -0.4021 0>}
 ]
 dyn_chin_tassel_01_l [{shapeType=Capsule, length=1.75, radius=1.5, center=<1.4754 0 0>, rotation=<rot -90 0 0>}]
 dyn_chin_tassel_01_r [{shapeType=Capsule, length=1.75, radius=1.5, center=<-1.4754 0 0>, rotation=<rot -90 0 0>}]
 dyn_chin_tassel_02_l [{shapeType=Capsule, length=3.5, radius=1.5, center=<2.9808 0 0>, rotation=<rot -90 0 0>}]
 dyn_chin_tassel_02_r [{shapeType=Capsule, length=3.5, radius=1.5, center=<-2.9808 0 0>, rotation=<rot -90 0 0>}]
 dyn_hair_03 [{shapeType=Capsule, length=3.5, radius=2.5, center=<1 0 0>, rotation=<rot 90 0 0>}]
 dyn_hair_05 [{shapeType=Capsule, length=2, radius=1.5, center=<0.5 0 0>, rotation=<rot 90 0 0>}]
 dyn_hair_06 [{shapeType=Capsule, length=3.5, radius=2.5, center=<1 0 0>, rotation=<rot 90 0 0>}]
 dyn_hair_07 [{shapeType=Capsule, length=3.5, radius=2.5, center=<1 0 0>, rotation=<rot 90 0 0>}]
 dyn_hair_08 [{shapeType=Capsule, length=5, radius=3, center=<2 0 0>, rotation=<rot 90 0 0>}]
 dyn_hair_09 [{shapeType=Capsule, length=2, radius=1.5, center=<0.5 0 0>, rotation=<rot 90 0 0>}]
 dyn_hair_10 [{shapeType=Capsule, length=5, radius=3, center=<2 0 0>, rotation=<rot 90 0 0>}]
 dyn_hair_12 [{shapeType=Capsule, length=5, radius=3, center=<2 0 0>, rotation=<rot 90 0 0>}]
 dyn_hair_13 [{shapeType=Capsule, length=2, radius=1.5, center=<0.5 0 0>, rotation=<rot 90 0 0>}]
 dyn_hair_14 [{shapeType=Capsule, length=3.5, radius=2.5, center=<1 0 0>, rotation=<rot 90 0 0>}]
 dyn_hair_15 [{shapeType=Capsule, length=4.3125, radius=3.2812, center=<0.7673 0 0>, rotation=<rot 90 0 0>}]
 dyn_hair_16 [{shapeType=Capsule, length=5, radius=3, center=<2 0 0>, rotation=<rot 90 0 0>}]
 head [{shapeType=Capsule, length=5, radius=5, rotation=<rot 90 0 0>}]
 spine_05 [{shapeType=Capsule, length=11.4208, radius=17.6105, center=<-3.6861 -2.7271 0>}]
}
constraints [
 {
  bone1Name=dyn_chin_guard
  bone2Name=head
  swing1LimitDegrees=45
  swing2LimitDegrees=1
  swing2Motion=Locked
  twistLimitDegrees=1
  twistMotion=Locked
 }
 {bone1Name=dyn_chin_tassel_01_l, bone2Name=dyn_chin_guard, swing1LimitDegrees=45, swing2LimitDegrees=45, twistLimitDegrees=1}
 {bone1Name=dyn_chin_tassel_02_l, bone2Name=dyn_chin_tassel_01_l, swing1LimitDegrees=30, swing2LimitDegrees=30, twistLimitDegrees=1}
 {bone1Name=dyn_chin_tassel_01_r, bone2Name=dyn_chin_guard, swing1LimitDegrees=45, swing2LimitDegrees=45, twistLimitDegrees=1}
 {bone1Name=dyn_chin_tassel_02_r, bone2Name=dyn_chin_tassel_01_r, swing1LimitDegrees=30, swing2LimitDegrees=30, twistLimitDegrees=1}
 {bone1Name=dyn_cheek_flap_l, bone2Name=head, swing1LimitDegrees=45, swing2LimitDegrees=5, twistLimitDegrees=5}
 {bone1Name=dyn_hair_03, bone2Name=head, swing1LimitDegrees=45, swing2LimitDegrees=45, twistLimitDegrees=10}
 {bone1Name=dyn_hair_05, bone2Name=head, swing1LimitDegrees=45, swing2LimitDegrees=45, twistLimitDegrees=10}
 {bone1Name=dyn_hair_06, bone2Name=head, swing1LimitDegrees=45, swing2LimitDegrees=45, twistLimitDegrees=10}
 {bone1Name=dyn_hair_07, bone2Name=head, swing1LimitDegrees=45, swing2LimitDegrees=45, twistLimitDegrees=10}
 {bone1Name=dyn_hair_08, bone2Name=head, swing1LimitDegrees=45, swing2LimitDegrees=45, twistLimitDegrees=10}
 {bone1Name=dyn_hair_09, bone2Name=head, swing1LimitDegrees=45, swing2LimitDegrees=45, twistLimitDegrees=10}
 {bone1Name=dyn_hair_10, bone2Name=head, swing1LimitDegrees=45, swing2LimitDegrees=45, twistLimitDegrees=10}
 {bone1Name=dyn_hair_12, bone2Name=head, swing1LimitDegrees=45, swing2LimitDegrees=45, twistLimitDegrees=10}
 {bone1Name=dyn_hair_13, bone2Name=head, swing1LimitDegrees=45, swing2LimitDegrees=45, twistLimitDegrees=10}
 {bone1Name=dyn_hair_14, bone2Name=head, swing1LimitDegrees=45, swing2LimitDegrees=45, twistLimitDegrees=10}
 {bone1Name=dyn_hair_15, bone2Name=head, swing1LimitDegrees=45, swing2LimitDegrees=45, twistLimitDegrees=10}
 {bone1Name=dyn_hair_16, bone2Name=head, swing1LimitDegrees=45, swing2LimitDegrees=45, twistLimitDegrees=10}
 {bone1Name=dyn_cheek_flap_r, bone2Name=head, swing1LimitDegrees=45, swing2LimitDegrees=5, twistLimitDegrees=5}
 {bone1Name=dyn_cheek_tassel_01_l, bone2Name=head, swing1LimitDegrees=45, swing2LimitDegrees=45, twistLimitDegrees=5}
 {bone1Name=dyn_cheek_tassel_02_l, bone2Name=dyn_cheek_tassel_01_l, swing1LimitDegrees=20, swing2LimitDegrees=20, twistLimitDegrees=10}
 {bone1Name=dyn_cheek_tassel_01_r, bone2Name=head, swing1LimitDegrees=45, swing2LimitDegrees=45, twistLimitDegrees=5}
 {bone1Name=dyn_cheek_tassel_02_r, bone2Name=dyn_cheek_tassel_01_r, swing1LimitDegrees=20, swing2LimitDegrees=20, twistLimitDegrees=10}
 {
  bone1Name=head
  bone2Name=spine_05
  swing1LimitDegrees=45
  swing1Motion=Locked
  swing2LimitDegrees=45
  swing2Motion=Locked
  twistLimitDegrees=45
  twistMotion=Locked
 }
]
